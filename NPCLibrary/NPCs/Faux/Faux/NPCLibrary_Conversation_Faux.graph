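# Conversation database for the Faux NPC, providing its dialogue in the voice of a character posing as a human.
# The Faux entry point leads to a Speech node with a default line and contextual lines for time since last match, first-ever conversation, and Bars held, prioritized by curve rows.
# The speech then branches through link nodes to shared conversations for NPC library quests, basic services, and Season 17 Weeks 3, 11 and 13.
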ConversationDatabase /NPCLibrary/NPCs/Faux/Faux/NPCLibrary_Conversation_Faux
entry_points [{tag=Conversation.Entry.Faux, destinations [#1]}]
nodes [
 ConversationEntryPointNode_7 {
  id=#1
  class=ConversationEntryPointNode
  outputs [#2]
  properties {entryTag {tagName=Conversation.Entry.Faux}}
 }
 Speech_C_4 {
  id=#2
  class=Speech_C
  outputs [
   #3
   #4
   #5
   #6
   #7
  ]
  properties {
   nodeName=""
   generalConfig {
    defaultMessage="If anyone asks, I come in peace."
    contextualMessages [
     {
      message="You totally remember meeting me, right? I've been here this whole time."
      contextRequirements [{requirement {class=FortControllerRequirement_CheckTimeSinceLastPlayedMatch}}]
      priority {curve {rowName=Tandem.Speech.Priority.WhileSincePlayed}}
     }
     {
      message="I'm just a normal human, meeting another normal human. Don't ask any questions."
      contextRequirements [{requirement {class=FortControllerRequirement_IsFirstEverConversationWithNPC}}]
      priority {curve {rowName=Tandem.Speech.Priority.FirstEverInteraction}}
     }
     {
      message="Ah yes... bars... You have a lot of them. And I am also a human, so this makes me... jealous?"
      contextRequirements [{requirement {class=FortControllerRequirement_HasBars}}]
      priority {curve {rowName=Tandem.Speech.Priority.HighTotalBars}}
     }
    ]
   }
  }
 }
 ConversationLinkNode_4 {id=#7, class=ConversationLinkNode, properties {remoteEntryTag {tagName=Conversation.Entry.Season17.Week13}}}
 ConversationLinkNode_2 {id=#6, class=ConversationLinkNode, properties {remoteEntryTag {tagName=Conversation.Entry.Season17.Week11}}}
 ConversationLinkNode_3 {id=#5, class=ConversationLinkNode, properties {remoteEntryTag {tagName=Conversation.Entry.Season17.Week3}}}
 ConversationLinkNode_0 {id=#4, class=ConversationLinkNode, properties {remoteEntryTag {tagName=Conversation.Entry.BasicServices}}}
 ConversationLinkNode_1 {id=#3, class=ConversationLinkNode, properties {remoteEntryTag {tagName=Conversation.Entry.NPCLibrary.Quests}}}
]
